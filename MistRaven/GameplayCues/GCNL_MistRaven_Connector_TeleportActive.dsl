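; Looping gameplay cue for MistRaven's connector teleport-active state, which drives a post-process screen effect.
; On looping start it creates and stores a dynamic instance of the teleport post-process material, then sets it as a full-weight blendable on its PostProcess component.
; It stamps the material's timeline parameters with a fixed duration and current game time, leaving animation to the material.
; No other event carries logic.
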(event EventBeginPlay)

(event EventOnLoopingStart(EitherVFX-preferthisversion)
  (params (MyTarget :type Actor%20Object%20Reference) (Parameters :type Gameplay%20Cue%20Parameters%20Structure%20%28by%20ref%29) (ParticleComponents :type Array%20of%20FXSystem%20Component%20Object%20References) (AudioComponents :type Array%20of%20Audio%20Component%20Object%20References))
  (bind _ppmid_mistraven_teleport (Variables|Default|GetPPMID_MistRaven_Teleport))
  (bind _returnvalue (Rendering|Material|CreateDynamicMaterialInstance 0 (Variables|Default|GetPPMI_MistRaven_Teleport) "None" "None" :!class /Script/Engine.KismetMaterialLibrary))
  (Variables|Default|SetPPMID_MistRaven_Teleport _returnvalue)
  (Variables|/Script/Engine.PostProcessComponent|SetSettings (Utilities|Struct|MakePostProcessSettings (Utilities|Struct|MakeWeightedBlendables (Utilities|Array|MakeArray (Utilities|Struct|MakeWeightedBlendable 1.0 (Variables|Default|GetPPMID_MistRaven_Teleport))))) (Variables|Default|GetPostProcess))
  (Rendering|Material|SetScalarParameterValue _ppmid_mistraven_teleport "TL_Duration" 1.0 :!class /Script/Engine.MaterialInstanceDynamic)
  (Rendering|Material|SetScalarParameterValue _ppmid_mistraven_teleport "TL_Time" (Utilities|Time|GetGameTimeinSeconds 0) :!class /Script/Engine.MaterialInstanceDynamic))

(fn ConstructionScript)
