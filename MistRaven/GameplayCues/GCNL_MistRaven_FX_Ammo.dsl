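; Looping gameplay cue for MistRaven that plays the GrimStrut loop audio and its sound indicator while active, with no visual effects in its graph despite the FX_Ammo name.
; On looping start, it spawns a first-person (1P) GrimStrut sound attached to the target's root and starts a looping Fort sound indicator at the cue actor's location.
; On removal, it cuts the spawned audio with a zero-length fade and stops the indicator.
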(event EventOnLoopingStart(EitherVFX-preferthisversion)
  (params (MyTarget :type Actor%20Object%20Reference) (Parameters :type Gameplay%20Cue%20Parameters%20Structure%20%28by%20ref%29) (ParticleComponents :type Array%20of%20FXSystem%20Component%20Object%20References) (AudioComponents :type Array%20of%20Audio%20Component%20Object%20References))
  (bind _returnvalue (Audio|SpawnSoundAttached "/MistRaven/Assets/Audio/MSS_MistRaven_GrimStrut_LP_1P.MSS_MistRaven_GrimStrut_LP_1P" (Variables|/Script/Engine.Actor|GetRootComponent MyTarget)))
  (SoundIndicator|StartSoundIndicatorLooping (Variables|Default|GetFortSoundIndicator) (Transformation|GetActorLocation) "/MistRaven/Assets/Audio/MistRaven_GrimStrut_Traversal_Loop_01.MistRaven_GrimStrut_Traversal_Loop_01" 1.0 0))

(event EventOnRemoval(EitherVFX-preferthisversion)
  (params (MyTarget :type Actor%20Object%20Reference) (Parameters :type Gameplay%20Cue%20Parameters%20Structure%20%28by%20ref%29) (ParticleComponents :type Array%20of%20FXSystem%20Component%20Object%20References) (AudioComponents :type Array%20of%20Audio%20Component%20Object%20References) (BurstCameraShakeInstance :type Camera%20Shake%20Base%20Object%20Reference) (BurstDecalInstance :type Decal%20Actor%20Object%20Reference))
  (Audio|Components|Audio|FadeOut (stale EventOnLoopingStart(EitherVFX-preferthisversion) _returnvalue) 0.0 0.0 "Linear" :!class /Script/Engine.AudioComponent)
  (SoundIndicator|StopSoundIndicatorLooping (Variables|Default|GetFortSoundIndicator)))

(fn ConstructionScript)
