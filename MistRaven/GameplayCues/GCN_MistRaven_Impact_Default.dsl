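; Burst gameplay cue that plays the default MistRaven impact sound at the hit location.
; OnBurst requires a valid effect causer and owner, then casts that owner to PlayerPawn_Athena.
; Casting the target to the same class chooses between player and default impact sounds.
; Whether the owning pawn is locally viewed selects the 1P or 3P variant of that sound.
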(fn OnBurst(EitherVFX-preferthisversion)
  (params (MyTarget :type Actor%20Object%20Reference) (Parameters :type Gameplay%20Cue%20Parameters%20Structure%20%28by%20ref%29) (ParticleComponents :type Array%20of%20FXSystem%20Component%20Object%20References) (AudioComponents :type Array%20of%20Audio%20Component%20Object%20References) (BurstCameraShakeInstance :type Camera%20Shake%20Base%20Object%20Reference) (BurstDecalInstance :type Decal%20Actor%20Object%20Reference))
  (bind (_normalizedmagnitude _rawmagnitude _effectcontext _matchedtagname _originaltag _aggregatedsourcetags _aggregatedtargettags _location _normal _instigator _effectcauser) (Ability|GameplayCue|BreakGameplayCueParameters Parameters))
  (bind _returnvalue (Actor|GetOwner _effectcauser))
  (Utilities|IsValid _effectcauser
    (:Is%20Valid
      (Utilities|IsValid _returnvalue
        (:Is%20Valid
          (bind _asplayer_pawn_athena (Utilities|Casting|CastToPlayerPawn_Athena _returnvalue))
          (Utilities|Casting|CastToPlayerPawn_Athena MyTarget
            (:then
              (Audio|PlaySoundatLocation _asplayer_pawn_athena (select (Pawn|IsLocallyViewed _asplayer_pawn_athena) (Variables|Default|Get1P%20Player%20Impact%20Sound) (Variables|Default|Get3P%20Player%20Impact%20Sound)) _location))
            (:CastFailed
              (Audio|PlaySoundatLocation _asplayer_pawn_athena (select (Pawn|IsLocallyViewed _asplayer_pawn_athena) (Variables|Default|Get1P%20Default%20Impact%20Sound) (Variables|Default|Get3P%20Default%20Impact%20Sound)) _location)))
          (return))))))
